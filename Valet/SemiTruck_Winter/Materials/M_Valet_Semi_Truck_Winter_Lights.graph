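# Lit material for the blinking light strips on the Valet winter semi truck, usable on skeletal meshes.
# Time scaled by BlinkSpeed, phase-shifted by the second UV set's U coordinate, drives a rounded sine for hard on/off blinking.
# This gates an E-texture green-channel mask, tinted by boosted diffuse color and scaled by LightStripBrightness.
# A shading-path switch remaps the glow through DeriveHDRfromLDR by default, while mobile uses the brightened value directly.
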Material /Valet/SemiTruck_Winter/Materials/M_Valet_Semi_Truck_Winter_Lights
graph {
 properties {bUsedWithMeshDeformer=true, bUsedWithSkeletalMesh=true, pixelDepthOffsetMode=PDOM_Legacy, usedShadingModels=Default Lit}
 nodes [
  n1 = Multiply(A=Time(), B=ScalarParameter() {defaultValue=1, parameterName=BlinkSpeed})
  n2 = ComponentMask(None=TextureCoordinate() {coordinateIndex=1}) {r=true}
  n3 = ConstantBiasScale(None=Sine(None=Add(A=@n1, B=@n2)))
  n4 = Multiply(
   A=Multiply(A=TextureSample() {texture=/Valet/SemiTruck_Winter/Textures/T_Vehicle_Semi_Truck_Winter_D}.RGB) {constB=2}
   B=Multiply(A=TextureSample() {texture=/Valet/SemiTruck_Winter/Textures/T_Vehicle_Semi_Truck_Winter_E}.G, B=Round(None=@n3))
  )
  n5 = Multiply(A=ScalarParameter() {defaultValue=1000, parameterName=LightStripBrightness}, B=@n4)
  ShadingPathSwitch(
   Default=MaterialFunctionCall(LDR Input=@n4, HDR Tint=@n5) {materialFunction=&DeriveHDRfromLDR}.RemappedRange
   Mobile=@n5
  ) {desc=Migrated from FeatureLevelSwitch}
  TextureSample() {samplerType=SAMPLERTYPE_Normal, texture=/Valet/SemiTruck_Winter/Textures/T_Vehicle_Semi_Truck_Winter_N}
  TextureSample() {samplerType=SAMPLERTYPE_Masks, texture=/Valet/SemiTruck_Winter/Textures/T_Vehicle_Semi_Truck_Winter_S}
 ]
}
